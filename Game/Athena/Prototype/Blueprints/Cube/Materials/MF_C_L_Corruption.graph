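# Material function that layers animated, glowing purple Cube corruption onto input material attributes for Loot Lake.
# MPC_Landscape LakeCorruption channels drive spread radius, strength and pulse; TurnedOn swaps the growing sphere mask and edge glow for full coverage.
# Flow-map-distorted electric noise and bump-offset runes with sine flicker form emissive, blending by camera distance.
# Masks also blacken base color and reduce specular, while quality and mobile switches trim costly detail.
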MaterialFunction /Game/Athena/Prototype/Blueprints/Cube/Materials/MF_C_L_Corruption
graph {nodes [
  Constant() {r=1}
  n1 = WorldPosition()
  n2 = VectorParameter() {parameterName=MaskOffset, defaultValue=<rgba 82980 -59350 0 35880>}
  Divide(A=ComponentMask(None=Subtract(A=@n1.XYZ, B=@n2.RGB)) {g=true, r=true}, B=@n2.A)
  n3 = ComponentMask(None=CollectionParameter() {collection=/Game/Athena/Environments/Landscape/MPC/MPC_Landscape, parameterName=LakeCorruption}) {b=true}
  n4 = Multiply(A=VectorParameter() {parameterName=PulseColor, defaultValue=<rgba 2 0.1738 1.8546 0>}.RGB, B=@n3)
  n5 = ShadingPathSwitch(Default=Fresnel() {baseReflectFraction=0.15, exponent=9}, Mobile=Constant() {r=0.35}) {desc=Migrated from FeatureLevelSwitch}
  n6 = Add(A=Multiply(A=@n4) {constB=0.2}, B=@n5) {constB=0.25}
  n7 = TextureCoordinate()
  n8 = Multiply(A=@n7, B=ScalarParameter() {defaultValue=4.2, parameterName=MacroTile})
  n9 = ComponentMask(None=TextureSample(UVs=Panner(Coordinate=@n8) {speedX=0.012, speedY=0.02}) {samplerType=SAMPLERTYPE_LinearColor, texture=/Game/Effects/Fort_Effects/Creatures/Elemental/Fire/MergedFireFlowMap}.RGB) {g=true, r=true}
  n10 = TextureSample() {texture=&/T_Loot_Lake_Mask}
  n11 = LinearInterpolate(Alpha=@n10.R) {constA=0.4, constB=0.7}
  n12 = ComponentMask(None=CollectionParameter() {collection=&MPC_Landscape, parameterName=LakeCorruption}) {r=true}
  n13 = SphereMask(
   A=ComponentMask(None=WorldPosition().XYZ) {g=true, r=true}
   B=ComponentMask(None=VectorParameter() {parameterName=CorruptionCenter, defaultValue=<rgba 63281 -49074 0 0>}.RGB) {g=true, r=true}
   Radius=Multiply(A=ScalarParameter() {defaultValue=28000, parameterName=MaxCorruptionRadius}, B=@n12)
   Hardness=LinearInterpolate(Alpha=@n12) {constB=0.4}
  ) {hardnessPercent=25}
  n14 = LinearInterpolate(Alpha=Add(A=Multiply(A=@n10.R) {constB=0.4}, B=@n13)) {constA=1.2, constB=0.6}
  n15 = Multiply(
   A=QualitySwitch(Default=Subtract(A=@n9) {constB=0.5}, Low=Constant())
   B=Add(A=@n3, B=StaticSwitchParameter(True=@n11, False=@n14) {parameterName=TurnedOn})
  )
  n16 = Add(A=QualitySwitch(Default=Multiply(A=@n15) {constB=0.3}, Low=Constant()), B=@n8)
  n17 = TextureSample(UVs=@n16) {samplerType=SAMPLERTYPE_LinearColor, texture=/Game/Characters/Player/Male/Medium/Bodies/M_MED_SRDrift/Textures/T_ElectricNoise}
  n18 = Multiply(
   A=VectorParameter() {parameterName=DistantGlowColor, defaultValue=<rgba 0.4855 0.0781 1 0>}.RGB
   B=LinearInterpolate(A=Add(A=@n17.G) {constB=0.3}, Alpha=TextureSample() {texture=&T_Loot_Lake_Mask}.R) {constB=0.8}
  )
  n19 = Constant() {r=0.4}
  n20 = Constant() {r=0.88}
  n21 = FunctionInput() {inputType=FunctionInput_MaterialAttributes}
  n22 = MaterialFunctionCall!(Material=@n21) {materialFunction=/Engine/Functions/MaterialLayerFunctions/MatLayerBlend_BreakNormal}
  n23 = MaterialFunctionCall(Float2=@n22.Normal) {materialFunction=&BreakOutFloat2Components}
  n24 = Add(A=Add(A=@n23.R, B=@n23.R)) {constB=0.15}
  n25 = Constant() {r=0.2}
  n26 = ComponentMask(None=@n22.Normal) {g=true, r=true}
  n27 = ScalarParameter() {defaultValue=0.2, parameterName=DistortRunes}
  n28 = Multiply(A=@n7, B=ScalarParameter() {defaultValue=70, parameterName=RuneTile})
  n29 = Add(A=QualitySwitch(Default=Multiply(A=@n26, B=@n27) {constB=0.1}, Low=Constant()), B=Add(A=@n28, B=@n15))
  n30 = ScalarParameter() {defaultValue=-8, parameterName=BumpOffset}
  n31 = TextureSample(UVs=QualitySwitch(Default=BumpOffset(Coordinate=@n29, Height=@n30), Low=@n29, Medium=@n29)) {samplerType=SAMPLERTYPE_Masks, texture=&/T_DFHeightMask_Tile}
  n32 = Clamp(None=Multiply(A=@n17.G) {constB=3})
  n33 = LinearInterpolate(A=@n31.R, B=@n31.R, Alpha=@n32)
  n34 = LinearInterpolate(A=ScalarParameter() {defaultValue=0.75, parameterName=MacroMaskMin}, Alpha=@n17.G) {constA=0.75}
  n35 = LinearInterpolate(Alpha=@n12) {constA=1, constB=6}
  n36 = Add(
   A=Multiply(A=Time(), B=ScalarParameter() {defaultValue=0.2, parameterName=RunePhaseSpeed}) {constB=25}
   B=StaticSwitchParameter(True=Multiply(A=@n34) {constB=3}, False=Multiply(A=@n35, B=@n13) {constB=5}) {parameterName=TurnedOn}
  )
  n37 = Add(A=@n31.B, B=@n36)
  n38 = Power(Base=ConstantBiasScale(None=Sine(None=@n37))) {constExponent=4}
  n39 = LinearInterpolate(Alpha=@n31.B) {constA=0.4}
  n40 = Add(A=@n31.B, B=Time())
  n41 = ConstantBiasScale(None=Sine(None=Multiply(A=@n39, B=@n40)) {period=0.09})
  n42 = VectorParameter() {parameterName=RuneFlickerColor, defaultValue=<rgba 0.875 1.1213 7 1>}
  n43 = VectorParameter() {parameterName=RuneColor, defaultValue=<rgba 2.287 0.5625 3 0>}
  n44 = LinearInterpolate(A=Add(A=@n4, B=Multiply(A=@n41, B=@n42.RGB)), B=@n43.RGB, Alpha=@n38) {constA=1}
  n45 = Multiply(A=Multiply(A=Multiply(A=@n33, B=@n38), B=QualitySwitch(Default=@n44, Low=@n43.RGB))) {constB=2}
  n46 = Add(
   A=Multiply(A=@n18, B=StaticSwitchParameter(True=@n19, False=@n20) {parameterName=TurnedOn}) {constB=0.27}
   B=Multiply(A=QualitySwitch(Default=@n24, Low=@n25), B=LinearInterpolate(B=@n45, Alpha=@n32))
  )
  n47 = MaterialFunctionCall(Distance=ScalarParameter() {defaultValue=7000, parameterName=FadeDistance}) {materialFunction=&Near_Camera_Fade}
  n48 = Multiply(
   A=Add(A=@n4)
   B=ShadingPathSwitch(Default=LinearInterpolate(A=@n46, B=@n18, Alpha=@n47.Result), Mobile=@n18) {desc=Migrated from FeatureLevelSwitch}
  )
  n49 = VectorParameter() {parameterName=DimGlow, defaultValue=<rgba 0.0118 0.0362 0.08 1>}
  n50 = OneMinus(None=@n47.Result)
  n51 = TextureSample() {texture=&T_Loot_Lake_Mask}
  n52 = Multiply(A=@n16, B=ScalarParameter() {defaultValue=2, parameterName=DetailTileMult}) {constB=1.7}
  n53 = ScalarParameter() {defaultValue=-2, parameterName=BumopOffset2}
  n54 = TextureSample(UVs=QualitySwitch(Default=BumpOffset(Coordinate=@n52, Height=@n53), Low=@n52, Medium=@n52)) {samplerType=SAMPLERTYPE_LinearColor, texture=&T_ElectricNoise}
  n55 = Multiply(A=@n54.G, B=Multiply(A=VectorParameter() {parameterName=DetailColor, defaultValue=<rgba 0.1764 0.0742 0.8 0>}.RGB, B=OneMinus(None=@n51.R)))
  n56 = Constant()
  n57 = QualitySwitch(Default=StaticSwitchParameter(True=Multiply(A=@n50, B=@n55), False=Constant()) {parameterName=TurnedOn}, Low=@n56, Medium=@n56)
  n58 = StaticSwitchParameter(True=QualitySwitch(Default=LinearInterpolate(A=@n34, Alpha=@n54.G) {constA=0.4}, Low=@n34), False=Multiply(A=@n34, B=@n13)) {parameterName=TurnedOn}
  n59 = LinearInterpolate(A=ConstantBiasScale(None=Sine(None=@n36) {desc=Radial ripples}), Alpha=@n58)
  n60 = Add(A=@n48, B=Multiply(A=@n59, B=Add(A=@n49.RGB, B=@n57))) {constB=0}
  n61 = Constant()
  n62 = TextureSample() {texture=&T_Loot_Lake_Mask}
  n63 = Add(A=@n58, B=Multiply(A=@n62.R) {constB=0.1})
  n64 = ScalarParameter() {defaultValue=1, parameterName=EdgeSineOffset}
  n65 = Power(Base=Sine(None=Add(A=@n63, B=@n64)), Exp=ScalarParameter() {defaultValue=12, parameterName=EdgeGlowPower}) {constExponent=3}
  n66 = Add(A=Multiply(A=@n15) {constB=4}, B=@n28)
  n67 = ScalarParameter() {defaultValue=0.2, parameterName=ChromaticTile}
  n68 = TextureSample(UVs=Panner(Coordinate=Multiply(A=@n66, B=@n67)) {speedX=0.037, speedY=0.032}) {texture=&T_Inky_Smoke_Tile_Chromatic}
  n69 = Constant() {r=0.4}
  n70 = Add(A=QualitySwitch(Default=Multiply(A=@n34, B=@n68.RGB), Low=@n69, Medium=@n69)) {constB=0.15}
  n71 = Multiply(
   A=Multiply(A=@n42.RGB, B=ScalarParameter() {defaultValue=3, parameterName=EdgeGlowBrightness})
   B=Multiply(A=@n59, B=Multiply(A=@n65, B=@n70))
  )
  n72 = Multiply(
   A=Multiply(A=StaticSwitchParameter(True=@n6, False=@n5) {parameterName=TurnedOn}) {constB=2.5}
   B=Add(
    A=ShadingPathSwitch(Default=@n60, Mobile=@n48) {desc=Migrated from FeatureLevelSwitch}
    B=StaticSwitchParameter(True=@n61, False=@n71) {parameterName=TurnedOn}
   )
  )
  n73 = Multiply(A=@n58, B=@n62.G)
  n74 = ScalarParameter() {defaultValue=1.2, parameterName=EmissiveMult}
  n75 = ComponentMask(None=CollectionParameter() {collection=&MPC_Landscape, parameterName=LakeCorruption}) {g=true}
  n76 = MaterialFunctionCall!(Material=@n21, Emissive=Multiply(A=Multiply(A=@n72, B=@n73), B=Multiply(A=@n74, B=@n75))) {materialFunction=/Engine/Functions/MaterialLayerFunctions/MatLayerBlend_Emissive}
  n77 = Constant()
  n78 = Constant() {r=1}
  n79 = Multiply(A=@n75, B=Multiply(A=@n62.G, B=StaticSwitchParameter(True=@n78, False=@n13) {parameterName=TurnedOn}))
  n80 = ScalarParameter() {defaultValue=0.25, parameterName=CorruptSpecularMult}
  FunctionOutput(None=MaterialFunctionCall!(
    Base Material=MaterialFunctionCall!(Material=@n76.Blended Material, NewBaseColor=@n77, Mask=@n79) {materialFunction=/Engine/Functions/MaterialLayerFunctions/MatLayerBlend_MultiplyBaseColor}.Blended Material
    SpecularMultiplier=LinearInterpolate(B=@n80, Alpha=@n79) {constA=1}
   ) {materialFunction=/Engine/Functions/MaterialLayerFunctions/MatLayerBlend_ModulateSpecular}.Blended Material)
  MaterialFunctionCall() {materialFunction=&DebugTimeSine}
  MaterialFunctionCall(Speed=Constant() {r=0.05}) {materialFunction=&DebugTimeSine}
 ]}
