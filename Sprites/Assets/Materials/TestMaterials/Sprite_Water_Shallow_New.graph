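# Shallow water surface material in a Sprites test folder, built on material attributes with world-space normals and Default Lit shading.
# ShallowWaterFunction supplies the surface; the default shading path adds a render-to-texture water pass with a constant flow direction, while mobile skips it.
# An optional CubeFX corruption switch follows, feeding the game-wide shared material function.
# Separate vertex-color math labeled WPO squares inverted red, optionally constant, and scales it by alpha, with no visible consumer.
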Material /Sprites/Assets/Materials/TestMaterials/Sprite_Water_Shallow_New
graph {
 properties {
  bTangentSpaceNormal=false
  bUseMaterialAttributes=true
  bUsedWithInstancedStaticMeshes=true
  bUsedWithMeshDeformer=true
  bUsedWithNanite=true
  bUsedWithSkeletalMesh=true
  pixelDepthOffsetMode=PDOM_Legacy
  usedShadingModels=Default Lit
 }
 nodes [
  n1 = MaterialFunctionCall!(Height Offset Amount, Texture mip bias amount, Ground texture mask) {materialFunction=/Game/WaterAndWind/Materials/Water/ShallowWaterFunction}
  n2 = ShadingPathSwitch(
   Default=MaterialFunctionCall!(Render To Texture Pass, WS Flow Map Dir=Constant2Vector() {g=0.5, r=0.5}, Opacity, Default Mat Att=@n1.MA Result) {materialFunction=/Game/WaterAndWind/Materials/Water/WaterRenderToTextureMaterialAtt}.Result
   Mobile=@n1.MA Result
  ) {desc=Migrated from FeatureLevelSwitch}
  MaterialFunctionCall!(In=StaticSwitchParameter(True=MaterialFunctionCall!(In=@n2) {materialFunction=/Game/Athena/Prototype/Blueprints/Cube/Materials/MF_C_L_Corruption}.Result, False=@n2) {parameterName=CubeFX}) {materialFunction=/Game/Athena/WinterArt/Materials/MF_GameWideShared}
  n3 = OneMinus(None=StaticSwitchParameter(True=VertexColor().R, False=Constant() {r=1}) {defaultValue=true, group=optimizations, parameterName=Evaluate Deep Water})
  n4 = Multiply(A=@n3, B=@n3)
  Multiply(A=VertexColor().A, B=Multiply(B=Clamp(None=@n4)) {desc=WPO}) {desc=wpo}
 ]
}
